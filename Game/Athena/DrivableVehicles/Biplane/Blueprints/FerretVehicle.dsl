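; Blueprint functions for FerretVehicle, the drivable Athena biplane, covering explosion tuning, player feedback, cooldown cues, pickup attachment and quest stats.
; Construction wires the overlap volume, aim reticle and boost meter material; cached hotfix scalars supply explosion and redeploy settings.
; Health, speed, boost and overheat feedback uses material or Niagara parameters, while impact magnitude scales camera shake and selects haptics.
; Touch aim assist offsets the player's reticle using the projected world-space reticle.
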(fn AddCooldownCueEvent
  (params (PlayerPawn :type Fort%20Player%20Pawn%20Object%20Reference))
  (bind _returnvalue (Environment|ApplyGlobalEnvironmentGameplayEffecttoActor PlayerPawn (Variables|Default|GetGE_Fired)))
  (Variables|Default|SetCooldownCueEvent _returnvalue)
  (bind _returnvalue_2 (Environment|ApplyGlobalEnvironmentGameplayEffecttoActor PlayerPawn (Variables|Default|GetGE_Boost)))
  (Variables|Default|SetBoostCueEvent _returnvalue_2))

(fn AttachWeaponToCart
  (params (Selection :type Integer))
  (bind (_returnvalue_roll _returnvalue_pitch _returnvalue_yaw) (Math|Rotator|BreakRotator (Transformation|GetActorRotation (Variables|Default|GetFortPickup))))
  (bind _skeletalmesh (Variables|Vehicle|GetSkeletalMesh))
  (bind _fortpickup (Variables|Default|GetFortPickup))
  (switch int Selection
    (:Default
      _)
    (:0
      _)
    (:1
      (Transformation|AttachActorToComponent _fortpickup _skeletalmesh "Weapon2" "SnapToTarget" "SnapToTarget" "SnapToTarget" true)
      (Transformation|SetActorRotation (Variables|Default|GetFortPickup) (Math|Rotator|MakeRotator (Math|Random|RandomFloatinRange (+ _returnvalue_roll -1e+02) (+ _returnvalue_roll -8e+01)) (Math|Random|RandomFloatinRange (+ _returnvalue_pitch -5.0) (+ _returnvalue_pitch 5.0)) (Math|Random|RandomFloatinRange (+ _returnvalue_yaw -1e+02) (+ _returnvalue_yaw -8e+01))) false :!label _label))
    (:2
      (Transformation|AttachActorToComponent _fortpickup _skeletalmesh "Weapon3" "SnapToTarget" "SnapToTarget" "SnapToTarget" true)
      (goto _label))
    (:3
      (Transformation|AttachActorToComponent _fortpickup _skeletalmesh "Weapon4" "SnapToTarget" "SnapToTarget" "SnapToTarget" true)
      (goto _label))
    (:4
      (Transformation|AttachActorToComponent _fortpickup _skeletalmesh "Weapon5" "SnapToTarget" "SnapToTarget" "SnapToTarget" true)
      (goto _label)))
  (Transformation|AttachActorToComponent _fortpickup _skeletalmesh "Weapon1" "SnapToTarget" "SnapToTarget" "SnapToTarget" true)
  (goto _label))

(fn CacheHotfixVariables
  ;[0.03491 0.1667 0.1262] Cache the hotfix values so they do not need to be looked up on every use.
  (Variables|Default|SetbValetCurieEnabled (ScalableFloat|GetValueasBool (Variables|Default|GetHotfixValetCurieEnabled)))
  (Variables|Default|SetExplosionDamageVehicles (ScalableFloat|GetValueatLevel (Variables|Default|GetHotfixExplosionDamageVehicles)))
  (Variables|Default|SetExplosionDamagePlayers (ScalableFloat|GetValueatLevel (Variables|Default|GetHotfixExplosionDamagePlayers)))
  (Variables|Default|SetExplosionDamageBuildings (ScalableFloat|GetValueatLevel (Variables|Default|GetHotfixExplosionDamageBuildings)))
  (Variables|Default|SetGliderRedeployOnExit (ScalableFloat|GetValueasBool (Variables|Default|GetHotfixGliderRedeployOnExit)))
  (Variables|Default|SetExplodeRadius (ScalableFloat|GetValueatLevel (Variables|Default|GetHotfixExplodeRadius)))
  (Variables|Default|SetExplodeVehicleImpulse (ScalableFloat|GetValueatLevel (Variables|Default|GetHotfixExplodeVehicleImpulse)))
  (Variables|Default|SetExplodePlayerImpulse (ScalableFloat|GetValueatLevel (Variables|Default|GetHotfixExplodePlayerImpulse))))

(fn CleanUpFXs
  ;[0.02957 0.1903 0.2188] Clean up any spawned audio and visual FXs.
  (Networking|CanExecuteCosmeticEvents
    (:True
      (bind _output_get (LocalVariables|SetAudioComponents (Actor|GetComponentsByClass "/Script/Engine.AudioComponent")))
      (Utilities|Array|AppendArray _output_get (Utilities|Array|MakeArray (Variables|Default|GetDeathExplosionSFXComp) (Variables|Default|GetDamageEngineSmokeSFXComp) (Variables|Default|GetDeathExplosionSFXComp)))
      (Audio|Components|Audio|FadeOut _output_get (Math|Float|SafeDivide (Variables|Default|GetFadeDuration) 2.0) 0.0 "Linear" :!class /Script/Engine.AudioComponent)
      (Components|Activation|Deactivate (Variables|Default|GetWreckFX)))))

(fn ComputeLaunchJumpVelocityZ
  (params (Target :type Fort%20Player%20Pawn%20Object%20Reference) (IsDrive :type Boolean))
  (bind _output_get (LocalVariables|SetLaunchJumpVelocityZ (select (Variables|GetIsDrive) (Variables|/Script/FortniteGame.FortVehicleConfigs|GetDriverExitInAirLaunchUpScalar (Variables|Vehicle|GetFortFerretVehicleConfigs)) (Variables|/Script/FortniteGame.FortVehicleConfigs|GetPassengerExitInAirLaunchUpScalar (Variables|Vehicle|GetFortFerretVehicleConfigs)))))
  (return _output_get))

(fn GetNormalizedMissingHealth
  ;[0.02295 0.1284 0.1458] Returns the normalized value of missing health.
  (return (Math|Float|Lerp 1.0 0.0 (Math|Float|SafeDivide (Class|FortVehicleInterface|GetHealth) (Class|FortVehicleInterface|GetMaxHealth)))))

(fn HandleUpdateDamageFXParams
  ;[0.01649 0.08333 0.0825] Lets the damage engine smoke FX know how fast the vehicle is moving, so it can adjust the smoke trail length accordingly.
  (Networking|CanExecuteCosmeticEvents
    (:True
      (bind _damageenginesmokefx (Variables|Default|GetDamageEngineSmokeFX)
        (:then
          (Niagara|SetNiagaraVariable(Float) _damageenginesmokefx "Vehicle_Speed" (Math|Float|SafeDivide (Vehicle|GetForwardSpeedKmh) (Variables|Default|GetMaxForwardSpeedKmH))))
        (:else)))))

(fn HideQuickBarForAbility
  (params (Hide :type Boolean))
  (bind _driverpawn (Variables|Default|GetDriverPawn))
  ;[0 0.9057 1] Hide/Show Quickbar in HUD
  (Utilities|IsValid _driverpawn
    (:Is%20Valid
      (bind _asfort_player_controller_gameplay (Utilities|Casting|CastToFortPlayerControllerGameplay (Pawn|GetController _driverpawn)))
      (HUD|SetHUDElementVisibilitywithContext _asfort_player_controller_gameplay (GameplayTags|MakeGameplayTagContainerfromTag "(TagName=\"HUD.Athena.HideforVehicle\")") Hide "None" ""))))

(fn OnRep_CorrectiveImpulse)

(fn OnRep_ImpulseVector
  (Audio|BroadcastAbilityImpactatLocation 0 self (Variables|Default|GetImpulseVector) 777.0))

(fn PlayHapticsOnHit
  (params (HitMagnitude :type Float%20%28double-precision%29) (Grind :type Boolean))
  (bind _validplayercontrollers (LocalVariables|GetValidPlayerControllers))
  (bind _passengerplayercontroller (Variables|Default|GetPassengerPlayerController))
  (bind _drivingplayercontroller (Variables|Default|GetDrivingPlayerController))
  (Utilities|FlowControl|Sequence
    (:then_0
      (Utilities|IsValid _drivingplayercontroller
        (:Is%20Valid
          (Utilities|Array|Add (LocalVariables|GetValidPlayerControllers) _drivingplayercontroller))))
    (:then_1
      (Utilities|IsValid _passengerplayercontroller
        (:Is%20Valid
          (Utilities|Array|Add (LocalVariables|GetValidPlayerControllers) _passengerplayercontroller))))
    (:then_2
      (if (> (Utilities|Array|Length (LocalVariables|GetValidPlayerControllers)) 0)
        (Utilities|FlowControl|Sequence
          (:then_0
            (if Grind
              (Game|Feedback|ClientPlayForceFeedback (LocalVariables|GetValidPlayerControllers) "/Game/Weapons/ForceFeedback/FF_Ninja_MantisLeap.FF_Ninja_MantisLeap" "None" false false false)))
          (:then_1
            (if (> (Variables|GetHitMagnitude) 1e+04)
              (Utilities|FlowControl|Sequence
                (:then_0
                  (LocalVariables|SetCameraShakeMagnitude (Math|Float|MapRangeClamped (Variables|GetHitMagnitude) 1.5e+04 5e+04 0.1 1.0))
                  (Utilities|IsValid (Variables|Default|GetDrivingPlayerController)
                    (:Is%20Valid
                      (Game|Feedback|ClientStartCameraShake :self (LocalVariables|GetValidPlayerControllers) :Shake "/Game/Athena/DrivableVehicles/CameraShake_HitImpact_GolfCart.CameraShake_HitImpact_GolfCart_C" :Scale (LocalVariables|GetCameraShakeMagnitude) :PlaySpace "CameraLocal")
                      (if (> (Variables|GetHitMagnitude) 2e+04)
                        (Game|Feedback|ClientPlayForceFeedback _validplayercontrollers "/Game/Weapons/ForceFeedback/FF_Generic_Impact_Near.FF_Generic_Impact_Near" "None" false false false)
                        (else
                          (Game|Feedback|ClientPlayForceFeedback _validplayercontrollers "/Game/Weapons/ForceFeedback/FF_Generic_Build.FF_Generic_Build" "None" false false false))))))
                (:then_1
                  (if (< (Math|Float|Absolute(Float) (Math|Vector|DotProduct (Math|Vector|Normalize (Vehicle|GetPreviousLinearVelocity)) "0, 0,1.000000")) 0.5)
                    (bind _output_get (LocalVariables|SetHitDelta (- (Variables|Vehicle|GetCachedSpeed) (Math|Vector|VectorLength (Transformation|GetComponentVelocity (Variables|Vehicle|GetSkeletalMesh))))))
                    (if (> _output_get (LocalVariables|GetCameraPushVelocityDelta))
                      (Game|Feedback|ClientStartCameraShake (LocalVariables|GetValidPlayerControllers) "/Game/Athena/DrivableVehicles/CameraShake_BigHit.CameraShake_BigHit_C" (Math|Float|MapRangeClamped (LocalVariables|GetHitDelta) 3e+02 9e+02 0.5 1.0) "UserDefined" (Math|Rotator|MakeRotfromX (Vehicle|GetPreviousLinearVelocity))))))))))))))

(fn QuestUpdate
  (params (FortPlayerPawn :type Fort%20Player%20Pawn%20Object%20Reference))
  (Utilities|FlowControl|Sequence
    (:then_0
      (Networking|SwitchHasAuthority
        (:Authority
          (Quest|SendComplexCustomStatEvent (Quests|GetQuestManager (Fortnite|GetFortPlayerControllerfromActor FortPlayerPawn) "Invalid") 0 "(GameplayTags=((TagName=\"Gameplay.Athena.Quests.UseBiplane\")))" "(GameplayTags=)" 1))))
    (:then_1
      (return))))

(fn RemoveCooldownCueEvent
  (params (PlayerPawn :type Fort%20Player%20Pawn%20Object%20Reference))
  (bind _returnvalue (Ability|GetAbilitySystemComponent PlayerPawn))
  (bind _returnvalue_2 (Ability|GetAbilitySystemComponent PlayerPawn))
  (Ability|SendGameplayEventtoActor PlayerPawn (LocalVariables|GetGE_FiredCooldown) (Utilities|Struct|MakeGameplayEventData :ContextHandle "()" :InstigatorTags "(GameplayTags=)" :TargetTags "(GameplayTags=)" :EventMagnitude 0.0 :TargetData "()"))
  (Utilities|IsValid _returnvalue_2
    (:Is%20Valid
      (GameplayEffects|RemoveActiveGameplayEffect _returnvalue_2 (Variables|Default|GetCooldownCueEvent) -1)
      (Ability|SendGameplayEventtoActor PlayerPawn (LocalVariables|GetGE_BoostCooldown) (Utilities|Struct|MakeGameplayEventData :ContextHandle "()" :InstigatorTags "(GameplayTags=)" :TargetTags "(GameplayTags=)" :EventMagnitude 0.0 :TargetData "()"))
      (Utilities|IsValid _returnvalue
        (:Is%20Valid
          (GameplayEffects|RemoveActiveGameplayEffect _returnvalue (Variables|Default|GetBoostCueEvent) -1))))))

(fn SendBoostTimeToMaterials
  (params (BoostTimeValueForMID :type Float%20%28double-precision%29))
  (bind _output_get (Variables|Customization|SetBoostTimeValueForMID BoostTimeValueForMID))
  (Rendering|Material|SetScalarParameterValueonMaterials (Variables|Vehicle|GetSkeletalMesh) "BoostTime" _output_get))

(fn SendDamageValueToMaterials
  (params (SavedDamageValueForMID :type Float%20%28double-precision%29))
  (bind _output_get (Variables|Customization|SetSavedDamageValueForMID SavedDamageValueForMID))
  (Rendering|Material|SetScalarParameterValueonMaterials (Variables|Vehicle|GetSkeletalMesh) "Damage" _output_get))

(fn SendHeadlightValueToMaterials
  (params (SavedHeadlightValueForMID :type Float%20%28double-precision%29))
  (bind _output_get (Variables|Customization|SetSavedHeadlightValueForMID SavedHeadlightValueForMID))
  (Rendering|Material|SetScalarParameterValueonMaterials (Variables|Vehicle|GetSkeletalMesh) "HeadLights" _output_get))

(fn SendOverheartValueToMaterials
  (params (OverheatValueForMID :type Float%20%28double-precision%29))
  (bind _skeletalmesh (Variables|Vehicle|GetSkeletalMesh))
  (Rendering|Material|SetScalarParameterValueonMaterials _skeletalmesh "WeaponGlow" OverheatValueForMID) ; Glow gun
  (Rendering|Material|SetScalarParameterValueonMaterials _skeletalmesh "GlowTime" (Utilities|Time|GetGameTimeinSeconds)) ; Start glow fade out
)

(fn StopTouchAimAssistReticleUpdate
  (params (Pawn :type Pawn%20Object%20Reference))
  (bind _returnvalue (Pawn|GetController Pawn))
  (if (Variables|Default|GetbCanUpdateTouchReticlePosition)
    (Variables|Default|SetbCanUpdateTouchReticlePosition false)
    (if (Pawn|IsLocalController _returnvalue)
      (bind _asfort_player_controller (Utilities|Casting|CastToFortPlayerController _returnvalue))
      (Class|FortAimAssist2DTargetManager|ResetTouchReticleOffset (GameInstanceSubsystems|GetFortAimAssist2D_TargetManager) _asfort_player_controller))))

(fn UpdateTouchAimAssistReticle
  (params (FortPC :type Fort%20Player%20Controller%20Object%20Reference))
  (bind _reticle (Variables|Default|GetReticle))
  (bind _returnvalue (Transformation|GetWorldLocation _reticle))
  (bind (_screenposition _returnvalue_2) (Viewport|ProjecttoWidgetPosition FortPC _returnvalue true))
  (bind (_x _y) (Math|Vector2D|BreakVector2D _screenposition))
  (bind (_sizex _sizey) (HUD|GetViewportSize FortPC))
  (if _returnvalue_2
    (Class|FortAimAssist2DTargetManager|ApplyTouchReticleOffset (GameInstanceSubsystems|GetFortAimAssist2D_TargetManager) FortPC (Math|Vector2D|MakeVector2D (- _x (* _sizex 0.5)) (- _y (* _sizey 0.5))))))

(fn ConstructionScript
  (Variables|FortAthenaVehicle|SetOverlapComponent (Variables|Default|GetOverlapVolume))
  (bind _returnvalue (Gun|GetAimPointLocal))
  (Transformation|SetRelativeLocation (Variables|Default|GetReticle) _returnvalue false true)
  (bind _returnvalue_2 (Rendering|Material|CreateDynamicMaterialInstance (Variables|Default|GetBoostMeter) 0 0 "None" :!class /Script/Engine.PrimitiveComponent))
  (Variables|Default|SetBoostMeterMID _returnvalue_2))
