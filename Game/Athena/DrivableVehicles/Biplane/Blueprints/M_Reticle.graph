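# Unlit masked material for the Biplane's aiming reticle, drawing a dot texture as a camera-facing sprite with responsive AA.
# The Sprite function scales by camera-to-object distance over a parameter, keeping apparent size roughly constant.
# Its world position offset feeds a CameraOffset with a huge, clamped amount, shifting the sprite toward the camera, likely to avoid occlusion.
# Texture alpha is multiplied by a strength parameter.
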Material /Game/Athena/DrivableVehicles/Biplane/Blueprints/M_Reticle
graph {
 properties {
  bDisableDepthTest=true
  bEnableResponsiveAA=true
  blendMode=BLEND_Masked
  pixelDepthOffsetMode=PDOM_Legacy
  shadingModel=MSM_Unlit
  usedShadingModels=Unlit
 }
 nodes [
  n1 = ObjectPositionWS()
  n2 = CameraPositionWS()
  n3 = MaterialFunctionCall!(
   Center Location
   0-1 UVs
   XY scale=Divide(A=Distance(A=@n1, B=@n2), B=ScalarParameter() {defaultValue=100, parameterName=Param_1})
   OverridePivot
   "Normals (Optional)"
  ) {materialFunction=&Sprite}
  n4 = WorldPosition()
  Add(
   A=@n3.World position offset
   B=MaterialFunctionCall(
    Offset Amount=Constant() {r=1000000}
    Clamp Padding=ScalarParameter() {defaultValue=100, parameterName=Param_5}
    World Position=Add(A=@n3.World position offset, B=@n4.XYZ)
   ) {materialFunction=/Engine/Functions/Engine_MaterialFunctions02/WorldPositionOffset/CameraOffset}.Clamped Camera Offset
  )
  Constant3Vector() {constant=<rgba 1 1 1 0>}
  Constant()
  n5 = Subtract(A=WorldPosition().XYZ, B=ObjectPositionWS())
  Multiply(A=@n5)
  Multiply(A=TextureSample() {texture=/Game/Athena/HUD/reticles/reticle_dot_new}.A, B=ScalarParameter() {defaultValue=2.8583, parameterName=Param_2})
  Normalize(VectorInput=@n5)
  TextureCoordinate()
 ]
}
