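; Santa variant of the drivable biplane vehicle Blueprint, which stops player cosmetic wraps from being applied to this plane.
; It overrides K2_ApplyCosmeticWrap to ignore the loaded wrap and only resend saved damage, headlight, boost time and overheat values to materials.
; A separate custom event repeats that material refresh.
; The construction script defers to the parent, while BeginPlay, Tick and overlap events are inactive stubs.
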(event EventBeginPlay
  (|CallParentFunction) ; This node is disabled and will not be called.\nDrag off pins to build functionality.
)

(event Collision|EventActorBeginOverlap
  (params (OtherActor :type Actor%20Object%20Reference)))

(event Custom|CustomEvent
  ;[1 1 1] Adding this so Santa Biplane no longer has player wraps applied.
  (CallFunction|SendDamageValuetoMaterials (Variables|Customization|GetSavedDamageValueForMID))
  (CallFunction|SendHeadlightValuetoMaterials (Variables|Customization|GetSavedHeadlightValueForMID))
  (CallFunction|SendBoostTimetoMaterials (Variables|Customization|GetBoostTimeValueForMID))
  (CallFunction|SendOverheartValuetoMaterials (Variables|Customization|GetOverheatValueForMID)))

(event EventK2_ApplyCosmeticWrap
  (params (LoadedWrap :type Athena%20Item%20Wrap%20Definition%20Object%20Reference))
  ;[1 1 1] Adding this so Santa Biplane no longer has player wraps applied.
  (CallFunction|SendDamageValuetoMaterials (Variables|Customization|GetSavedDamageValueForMID))
  (CallFunction|SendHeadlightValuetoMaterials (Variables|Customization|GetSavedHeadlightValueForMID))
  (CallFunction|SendBoostTimetoMaterials (Variables|Customization|GetBoostTimeValueForMID))
  (CallFunction|SendOverheartValuetoMaterials (Variables|Customization|GetOverheatValueForMID)))

(event EventTick
  (params (DeltaSeconds :type Float%20%28single-precision%29)))

(fn ConstructionScript
  (|CallParentFunction))
