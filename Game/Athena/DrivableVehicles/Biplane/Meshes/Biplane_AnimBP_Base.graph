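# Base animation Blueprint for the drivable Biplane, posing its airframe, cockpit gauges and landing gear.
# A Locomotion state machine handles steering blendspaces and boost sequences, with an additive Boost state machine layered on top.
# A bone-controller chain then applies variable-driven transforms to gauges, steering, wheels, control surfaces and propeller, with paired Look At nodes aligning axles and shocks.
# The event graph is empty, and the LandingGearAnim player is unconnected to the output.
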AnimBlueprint /Game/Athena/DrivableVehicles/Biplane/Meshes/Biplane_AnimBP_Base
target_skeleton=&/Biplane_Skeleton
graphs [
 AnimGraph {nodes [
   Local To Component(LocalPose=Sequence Player 'LandingGearAnim'() {sequence=&/LandingGearAnim})
   AnimGraphNode_StateMachine_0 = Locomotion() {is=title AnimGraphNode_StateMachine}
   AnimGraphNode_StateMachine_1 = Boost() {is=title AnimGraphNode_StateMachine}
   n1 = "Transform (Modify) Bone - Bone: booster_pin"!(
    ComponentPose=Local To Component(LocalPose=Apply Additive(Base=@AnimGraphNode_StateMachine_0, Additive=@AnimGraphNode_StateMachine_1) {alphaBoolBlend {blendOption=Linear}})
    Alpha=1
    Rotation=Get BoostPercent() -> (BoostPercent:Rotator)
   ) -> (Pose) {
    rotationMode=BMM_Replace
    rotationSpace=BCS_BoneSpace
    scaleSpace=BCS_BoneSpace
    translationSpace=BCS_BoneSpace
    alphaBoolBlend {blendOption=Linear}
    boneToModify {boneName=booster_pin}
   }
   n2 = "Transform (Modify) Bone - Bone: speedometer_pin"!(ComponentPose=@n1, Alpha=1, Rotation=Get FerretVehicleSpeed() -> (FerretVehicleSpeed:Rotator)) -> (Pose) {
    rotationMode=BMM_Replace
    rotationSpace=BCS_BoneSpace
    scaleSpace=BCS_BoneSpace
    translationSpace=BCS_BoneSpace
    alphaBoolBlend {blendOption=Linear}
    boneToModify {boneName=speedometer_pin}
   }
   n3 = "Transform (Modify) Bone - Bone: altimiter_pin"!(ComponentPose=@n2, Alpha=1, Rotation=Get Altitude() -> (Altitude:Rotator)) -> (Pose) {
    rotationMode=BMM_Replace
    rotationSpace=BCS_BoneSpace
    scaleSpace=BCS_BoneSpace
    translationSpace=BCS_BoneSpace
    alphaBoolBlend {blendOption=Linear}
    boneToModify {boneName=altimiter_pin}
   }
   n4 = "Transform (Modify) Bone - Bone: wheel_trans_fr_l"!(ComponentPose=@n3, Alpha=1, Translation=Get WheelPositionFL() -> (WheelPositionFL:Vector)) -> (Pose) {
    rotationSpace=BCS_BoneSpace
    scaleSpace=BCS_BoneSpace
    translationMode=BMM_Replace
    translationSpace=BCS_BoneSpace
    alphaBoolBlend {blendOption=Linear}
    boneToModify {boneName=wheel_trans_fr_l}
   }
   n5 = "Transform (Modify) Bone - Bone: wheel_trans_fr_r"!(ComponentPose=@n4, Alpha=1, Translation=NegateVector!(A:Vector=Get WheelPositionFR() -> (WheelPositionFR:Vector)) -> (ReturnValue:Vector)) -> (Pose) {
    rotationSpace=BCS_BoneSpace
    scaleSpace=BCS_BoneSpace
    translationMode=BMM_Replace
    translationSpace=BCS_BoneSpace
    alphaBoolBlend {blendOption=Linear}
    boneToModify {boneName=wheel_trans_fr_r}
   }
   n6 = "Transform (Modify) Bone - Bone: wheel_bk_steering"!(ComponentPose=@n5, Alpha=1, Rotation=Get WheelRotationBR() -> (WheelRotationBR:Rotator)) -> (Pose) {
    rotationMode=BMM_Replace
    rotationSpace=BCS_BoneSpace
    translationSpace=BCS_BoneSpace
    alphaBoolBlend {blendOption=Linear}
    boneToModify {boneName=wheel_bk_steering}
   }
   n7 = "Transform (Modify) Bone - Bone: wheel_bk_trans"!(ComponentPose=@n6, Alpha=1, Translation=Get AxlePositionB() -> (AxlePositionB:Vector)) -> (Pose) {
    rotationSpace=BCS_BoneSpace
    scaleSpace=BCS_BoneSpace
    translationMode=BMM_Replace
    translationSpace=BCS_BoneSpace
    alphaBoolBlend {blendOption=Linear}
    boneToModify {boneName=wheel_bk_trans}
   }
   n8 = "Transform (Modify) Bone - Bone: wheel_spin_fr_r"!(ComponentPose=@n7, Alpha=1, Rotation=Get WheelSpinFR() -> (WheelSpinFR:Rotator)) -> (Pose) {rotationMode=BMM_Replace, rotationSpace=BCS_BoneSpace, alphaBoolBlend {blendOption=Linear}, boneToModify {boneName=wheel_spin_fr_r}}
   n9 = "Transform (Modify) Bone - Bone: wheel_spin_fr_l"!(ComponentPose=@n8, Alpha=1, Rotation=Get WheelSpinFL() -> (WheelSpinFL:Rotator)) -> (Pose) {rotationMode=BMM_Replace, rotationSpace=BCS_BoneSpace, alphaBoolBlend {blendOption=Linear}, boneToModify {boneName=wheel_spin_fr_l}}
   n10 = "Transform (Modify) Bone - Bone: wheel_bk_spin"!(ComponentPose=@n9, Alpha=1, Rotation=Get WheelSpinBR() -> (WheelSpinBR:Rotator)) -> (Pose) {rotationMode=BMM_Replace, rotationSpace=BCS_BoneSpace, alphaBoolBlend {blendOption=Linear}, boneToModify {boneName=wheel_bk_spin}}
   n11 = "Transform (Modify) Bone - Bone: steering_wheel"!(ComponentPose=@n10, Alpha=1, Rotation=Get SteeringRotation() -> (SteeringRotation:Rotator)) -> (Pose) {rotationMode=BMM_Replace, rotationSpace=BCS_BoneSpace, alphaBoolBlend {blendOption=Linear}, boneToModify {boneName=steering_wheel}}
   n12 = Look At - Bone: axel_fr_l(ComponentPose=@n11, Alpha=1) {
    alphaBoolBlend {blendOption=Linear}
    boneToModify {boneName=axel_fr_l}
    lookAtLocation {x=0}
    lookAtTarget {boneReference {boneName=axel_base_fr_l}}
    lookAt_Axis {axis=<1 0>}
    lookUp_Axis {axis {x=0, z=1}}
   }
   n13 = Look At - Bone: axel_base_fr_l(ComponentPose=@n12, Alpha=1) {
    alphaBoolBlend {blendOption=Linear}
    boneToModify {boneName=axel_base_fr_l}
    lookAtLocation {x=0}
    lookAtTarget {boneReference {boneName=axel_fr_l}}
    lookAt_Axis {axis=<-1 0>}
    lookUp_Axis {axis {x=0, z=1}}
   }
   n14 = Look At - Bone: axel_fr_r(ComponentPose=@n13, Alpha=1) {
    alphaBoolBlend {blendOption=Linear}
    boneToModify {boneName=axel_fr_r}
    lookAtLocation {x=0}
    lookAtTarget {boneReference {boneName=axel_base_fr_r}}
    lookAt_Axis {axis=<-1 0>}
    lookUp_Axis {axis {x=0, z=1}}
   }
   n15 = Look At - Bone: axel_base_fr_r(ComponentPose=@n14, Alpha=1) {
    alphaBoolBlend {blendOption=Linear}
    boneToModify {boneName=axel_base_fr_r}
    lookAtLocation {x=0}
    lookAtTarget {boneReference {boneName=axel_fr_r}}
    lookAt_Axis {axis=<1 0>}
    lookUp_Axis {axis {x=0, z=1}}
   }
   n16 = Look At - Bone: shock_top_fr_l(ComponentPose=@n15, Alpha=1) {
    alphaBoolBlend {blendOption=Linear}
    boneToModify {boneName=shock_top_fr_l}
    lookAtLocation {x=0}
    lookAtTarget {boneReference {boneName=shock_bott_fr_l}}
    lookAt_Axis {axis=<1 0>}
    lookUp_Axis {axis {x=0, z=1}}
   }
   n17 = Look At - Bone: shock_bott_fr_l(ComponentPose=@n16, Alpha=1) {
    alphaBoolBlend {blendOption=Linear}
    boneToModify {boneName=shock_bott_fr_l}
    lookAtLocation {x=0}
    lookAtTarget {boneReference {boneName=shock_top_fr_l}}
    lookAt_Axis {axis=<-1 0>}
    lookUp_Axis {axis {x=0, z=1}}
   }
   n18 = Look At - Bone: shock_top_fr_r(ComponentPose=@n17, Alpha=1) {
    alphaBoolBlend {blendOption=Linear}
    boneToModify {boneName=shock_top_fr_r}
    lookAtLocation {x=0}
    lookAtTarget {boneReference {boneName=shock_bott_fr_r}}
    lookAt_Axis {axis=<-1 0>}
    lookUp_Axis {axis {x=0, z=1}}
   }
   n19 = Look At - Bone: shock_bott_fr_r(ComponentPose=@n18, Alpha=1) {
    alphaBoolBlend {blendOption=Linear}
    boneToModify {boneName=shock_bott_fr_r}
    lookAtLocation {x=0}
    lookAtTarget {boneReference {boneName=shock_top_fr_r}}
    lookAt_Axis {axis=<1 0>}
    lookUp_Axis {axis {x=0, z=1}}
   }
   n20 = Look At - Bone: wheel_bk_shock(ComponentPose=@n19, Alpha=1) {
    alphaBoolBlend {blendOption=Linear}
    boneToModify {boneName=wheel_bk_shock}
    lookAtLocation {x=0}
    lookAtTarget {boneReference {boneName=wheel_bk_shock_bott}}
    lookAt_Axis {axis=<1 0>}
    lookUp_Axis {axis {x=0, z=1}}
   }
   n21 = Look At - Bone: wheel_bk_shock_bott(ComponentPose=@n20, Alpha=1) {
    alphaBoolBlend {blendOption=Linear}
    boneToModify {boneName=wheel_bk_shock_bott}
    lookAtLocation {x=0}
    lookAtTarget {boneReference {boneName=wheel_bk_shock}}
    lookAt_Axis {axis=<-1 0>}
    lookUp_Axis {axis {x=0, z=1}}
   }
   n22 = "Transform (Modify) Bone - Bone: rudder"!(ComponentPose=@n21, Alpha=1, Rotation=Get RudderAngle() -> (RudderAngle:Rotator)) -> (Pose) {rotationMode=BMM_Replace, rotationSpace=BCS_BoneSpace, alphaBoolBlend {blendOption=Linear}, boneToModify {boneName=rudder}}
   n23 = "Transform (Modify) Bone - Bone: elevator"!(ComponentPose=@n22, Alpha=1, Rotation=Get ElevatorAngle() -> (ElevatorAngle:Rotator)) -> (Pose) {rotationMode=BMM_Replace, rotationSpace=BCS_BoneSpace, alphaBoolBlend {blendOption=Linear}, boneToModify {boneName=elevator}}
   n24 = Get AileronAngle() -> (AileronAngle:Rotator)
   n25 = "Transform (Modify) Bone - Bone: aieleron_l"!(ComponentPose=@n23, Alpha=1, Rotation=Invert Rotator!(A:Rotator=@n24) -> (ReturnValue:Rotator)) -> (Pose) {rotationMode=BMM_Replace, rotationSpace=BCS_BoneSpace, alphaBoolBlend {blendOption=Linear}, boneToModify {boneName=aieleron_l}}
   n26 = "Transform (Modify) Bone - Bone: aieleron_r"!(ComponentPose=@n25, Alpha=1, Rotation=@n24) -> (Pose) {rotationMode=BMM_Replace, rotationSpace=BCS_BoneSpace, alphaBoolBlend {blendOption=Linear}, boneToModify {boneName=aieleron_r}}
   n27 = Get FlapAngle() -> (FlapAngle:Rotator)
   n28 = "Transform (Modify) Bone - Bone: flap_l"!(ComponentPose=@n26, Alpha=1, Rotation=@n27) -> (Pose) {rotationMode=BMM_Replace, rotationSpace=BCS_BoneSpace, alphaBoolBlend {blendOption=Linear}, boneToModify {boneName=flap_l}}
   n29 = "Transform (Modify) Bone - Bone: flap_r"!(ComponentPose=@n28, Alpha=1, Rotation=@n27) -> (Pose) {rotationMode=BMM_Replace, rotationSpace=BCS_BoneSpace, alphaBoolBlend {blendOption=Linear}, boneToModify {boneName=flap_r}}
   n30 = "Transform (Modify) Bone - Bone: prop"!(ComponentPose=@n29, Alpha=1, Rotation=Get PropellorAngle() -> (PropellorAngle:Rotator)) -> (Pose) {rotationMode=BMM_Replace, rotationSpace=BCS_BoneSpace, alphaBoolBlend {blendOption=Linear}, boneToModify {boneName=prop}}
   Output Pose(Result=Component To Local(ComponentPose=@n30)) {layerGroup=None}
  ]}
 Locomotion {
  entry=MovementLoop
  path=AnimGraph.AnimGraphNode_StateMachine_0.Locomotion
  ids=id
  nodes [
   AnimStateNode_0 = MovementLoop() {is=title AnimStateNode}
   AnimStateNode_3 = Ground() {is=title AnimStateNode}
   AnimStateNode_5 = BoostingStart() {is=title AnimStateNode}
   AnimStateNode_6 = BoostingLoop() {is=title AnimStateNode}
   AnimStateNode_7 = BoostingEnd() {is=title AnimStateNode}
   AnimStateTransitionNode_12 = AnimStateTransitionNode() {bAutomaticRuleBasedOnSequencePlayerInState=true}
   AnimStateTransitionNode_13 = AnimStateTransitionNode()
   AnimStateTransitionNode_14 = AnimStateTransitionNode() {bAutomaticRuleBasedOnSequencePlayerInState=true}
   AnimStateTransitionNode_15 = AnimStateTransitionNode() {crossfadeDuration=0.1}
   AnimStateTransitionNode_4 = AnimStateTransitionNode()
   AnimStateTransitionNode_5 = AnimStateTransitionNode() {crossfadeDuration=0.4}
  ]
  links [
   {
    dst=MovementLoop
    rule=AnimGraph.AnimGraphNode_StateMachine_0.Locomotion.AnimStateTransitionNode_14.Transition
    src=BoostingEnd
    via=AnimStateTransitionNode_14
   }
   {
    dst=BoostingEnd
    rule=AnimGraph.AnimGraphNode_StateMachine_0.Locomotion.AnimStateTransitionNode_13.Transition
    src=BoostingLoop
    via=AnimStateTransitionNode_13
   }
   {
    dst=BoostingLoop
    rule=AnimGraph.AnimGraphNode_StateMachine_0.Locomotion.AnimStateTransitionNode_12.Transition
    src=BoostingStart
    via=AnimStateTransitionNode_12
   }
   {
    dst=MovementLoop
    rule=AnimGraph.AnimGraphNode_StateMachine_0.Locomotion.AnimStateTransitionNode_5.Transition
    src=Ground
    via=AnimStateTransitionNode_5
   }
   {
    dst=BoostingStart
    rule=AnimGraph.AnimGraphNode_StateMachine_0.Locomotion.AnimStateTransitionNode_15.Transition
    src=MovementLoop
    via=AnimStateTransitionNode_15
   }
   {
    dst=Ground
    rule=AnimGraph.AnimGraphNode_StateMachine_0.Locomotion.AnimStateTransitionNode_4.Transition
    src=MovementLoop
    via=AnimStateTransitionNode_4
   }
  ]
 }
 MovementLoop {
  path=AnimGraph.AnimGraphNode_StateMachine_0.Locomotion.AnimStateNode_0.MovementLoop
  nodes [Output Animation Pose(Result=Blendspace Player 'FerretVehicle_SteeringWheel_Ground_BS'(X=Get AbsSteerAngle() -> (AbsSteerAngle:float), Y=Get SteerPitchAngle() -> (SteerPitchAngle:float)) {blendSpace=/Game/Animation/Game/MainPlayer/Vehicles/FerretVehicle/Parts/FerretVehicle_SteeringWheel_Ground_BS})]
 }
 Ground {
  path=AnimGraph.AnimGraphNode_StateMachine_0.Locomotion.AnimStateNode_3.Ground
  nodes [
   n1 = Get AbsSteerAngle()
   n2 = Get SteerPitchAngle()
   Output Animation Pose(Result=Blendspace Player 'FerretVehicle_SteeringWheel_Ground_BS'(X=float * float(A=@n1, B=0.5), Y=float * float(A=@n2, B=0.5)) {blendSpace=&FerretVehicle_SteeringWheel_Ground_BS})
  ]
 }
 BoostingStart {
  path=AnimGraph.AnimGraphNode_StateMachine_0.Locomotion.AnimStateNode_5.BoostingStart
  nodes [Output Animation Pose(Result=Sequence Player 'FerretVehicle_Drive_S_Start_Vehicle'() {bLoopAnimation=false, sequence=/Game/Animation/Game/MainPlayer/Vehicles/FerretVehicle/Parts/FerretVehicle_Drive_S_Start_Vehicle})]
 }
 BoostingLoop {
  path=AnimGraph.AnimGraphNode_StateMachine_0.Locomotion.AnimStateNode_6.BoostingLoop
  nodes [Output Animation Pose(Result=Sequence Player 'FerretVehicle_Drive_S_Vehicle'() {sequence=/Game/Animation/Game/MainPlayer/Vehicles/FerretVehicle/Parts/FerretVehicle_Drive_S_Vehicle})]
 }
 BoostingEnd {
  path=AnimGraph.AnimGraphNode_StateMachine_0.Locomotion.AnimStateNode_7.BoostingEnd
  nodes [Output Animation Pose(Result=Sequence Player 'FerretVehicle_Drive_S_Stop_Vehicle'() {bLoopAnimation=false, sequence=/Game/Animation/Game/MainPlayer/Vehicles/FerretVehicle/Parts/FerretVehicle_Drive_S_Stop_Vehicle})]
 }
 Transition {path=AnimGraph.AnimGraphNode_StateMachine_0.Locomotion.AnimStateTransitionNode_12.Transition, nodes [Result()]}
 Transition {
  path=AnimGraph.AnimGraphNode_StateMachine_0.Locomotion.AnimStateTransitionNode_13.Transition
  nodes [Result(bCanEnterTransition=NOT Boolean(A=Get bIsBoosting() -> (bIsBoosting:bool)))]
 }
 Transition {path=AnimGraph.AnimGraphNode_StateMachine_0.Locomotion.AnimStateTransitionNode_14.Transition, nodes [Result()]}
 Transition {
  path=AnimGraph.AnimGraphNode_StateMachine_0.Locomotion.AnimStateTransitionNode_15.Transition
  nodes [Result(bCanEnterTransition=Get bIsBoosting())]
 }
 Transition {
  path=AnimGraph.AnimGraphNode_StateMachine_0.Locomotion.AnimStateTransitionNode_4.Transition
  nodes [Result(bCanEnterTransition=Get bOnGround() -> (bOnGround:bool))]
 }
 Transition {
  path=AnimGraph.AnimGraphNode_StateMachine_0.Locomotion.AnimStateTransitionNode_5.Transition
  nodes [Result(bCanEnterTransition=NOT Boolean(A=Get bOnGround()))]
 }
 Boost {
  entry=BoostOff
  path=AnimGraph.AnimGraphNode_StateMachine_1.Boost
  ids=id
  nodes [
   AnimStateNode_0 = BoostOff() {is=title AnimStateNode}
   AnimStateNode_2 = BoostOnHold() {is=title AnimStateNode}
   AnimStateNode_3 = BoostReady() {is=title AnimStateNode}
   AnimStateNode_4 = BoostFade() {is=title AnimStateNode}
   AnimStateNode_5 = BoostOn() {is=title AnimStateNode}
   AnimStateTransitionNode_3 = AnimStateTransitionNode()
   AnimStateTransitionNode_4 = AnimStateTransitionNode() {bAutomaticRuleBasedOnSequencePlayerInState=true}
   AnimStateTransitionNode_7 = AnimStateTransitionNode() {bAutomaticRuleBasedOnSequencePlayerInState=true}
   AnimStateTransitionNode_8 = AnimStateTransitionNode()
   AnimStateTransitionNode_9 = AnimStateTransitionNode()
  ]
  links [
   {
    dst=BoostOff
    rule=AnimGraph.AnimGraphNode_StateMachine_1.Boost.AnimStateTransitionNode_7.Transition
    src=BoostFade
    via=AnimStateTransitionNode_7
   }
   {
    dst=BoostReady
    rule=AnimGraph.AnimGraphNode_StateMachine_1.Boost.AnimStateTransitionNode_3.Transition
    src=BoostOff
    via=AnimStateTransitionNode_3
   }
   {
    dst=BoostOnHold
    rule=AnimGraph.AnimGraphNode_StateMachine_1.Boost.AnimStateTransitionNode_8.Transition
    src=BoostOn
    via=AnimStateTransitionNode_8
   }
   {
    dst=BoostFade
    rule=AnimGraph.AnimGraphNode_StateMachine_1.Boost.AnimStateTransitionNode_9.Transition
    src=BoostOnHold
    via=AnimStateTransitionNode_9
   }
   {
    dst=BoostOn
    rule=AnimGraph.AnimGraphNode_StateMachine_1.Boost.AnimStateTransitionNode_4.Transition
    src=BoostReady
    via=AnimStateTransitionNode_4
   }
  ]
 }
 BoostOff {
  path=AnimGraph.AnimGraphNode_StateMachine_1.Boost.AnimStateNode_0.BoostOff
  nodes [Output Animation Pose(Result="Sequence Player (additive) 'FerretVehicle_Engine_Boost_Off_Vehicle'"() {sequence=/Game/Animation/Game/MainPlayer/Vehicles/FerretVehicle/Parts/FerretVehicle_Engine_Boost_Off_Vehicle})]
 }
 BoostOnHold {
  path=AnimGraph.AnimGraphNode_StateMachine_1.Boost.AnimStateNode_2.BoostOnHold
  nodes [Output Animation Pose(Result="Sequence Player (additive) 'FerretVehicle_Engine_Boost_On_Vehicle'"() {sequence=/Game/Animation/Game/MainPlayer/Vehicles/FerretVehicle/Parts/FerretVehicle_Engine_Boost_On_Vehicle})]
 }
 BoostReady {
  path=AnimGraph.AnimGraphNode_StateMachine_1.Boost.AnimStateNode_3.BoostReady
  nodes [
   n1 = "Sequence Player (additive) 'FerretVehicle_Engine_Boost_Ready_Vehicle'"() {bLoopAnimation=false, sequence=/Game/Animation/Game/MainPlayer/Vehicles/FerretVehicle/Parts/FerretVehicle_Engine_Boost_Ready_Vehicle}
   Output Animation Pose(Result=@n1)
  ]
 }
 BoostFade {
  path=AnimGraph.AnimGraphNode_StateMachine_1.Boost.AnimStateNode_4.BoostFade
  nodes [
   n1 = "Sequence Player (additive) 'FerretVehicle_Engine_Boost_Fading_Vehicle'"() {bLoopAnimation=false, sequence=/Game/Animation/Game/MainPlayer/Vehicles/FerretVehicle/Parts/FerretVehicle_Engine_Boost_Fading_Vehicle}
   Output Animation Pose(Result=@n1)
  ]
 }
 BoostOn {
  path=AnimGraph.AnimGraphNode_StateMachine_1.Boost.AnimStateNode_5.BoostOn
  nodes [Output Animation Pose(Result="Sequence Player (additive) 'FerretVehicle_Engine_Boost_On_Vehicle'"() {sequence=&FerretVehicle_Engine_Boost_On_Vehicle})]
 }
 Transition {
  path=AnimGraph.AnimGraphNode_StateMachine_1.Boost.AnimStateTransitionNode_3.Transition
  nodes [Result(bCanEnterTransition=Get bIsBoostReady() -> (bIsBoostReady:bool))]
 }
 Transition {path=AnimGraph.AnimGraphNode_StateMachine_1.Boost.AnimStateTransitionNode_4.Transition, nodes [Result()]}
 Transition {path=AnimGraph.AnimGraphNode_StateMachine_1.Boost.AnimStateTransitionNode_7.Transition, nodes [Result()]}
 Transition {
  path=AnimGraph.AnimGraphNode_StateMachine_1.Boost.AnimStateTransitionNode_8.Transition
  nodes [Result(bCanEnterTransition=NOT Boolean(A=Get bIsBoostReady()))]
 }
 Transition {
  path=AnimGraph.AnimGraphNode_StateMachine_1.Boost.AnimStateTransitionNode_9.Transition
  nodes [Result(bCanEnterTransition=NOT Boolean(A=Get bIsBoosting()))]
 }
 EventGraph {links [], nodes []}
]
variables []
