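# Dynamics physics asset for the drivable Biplane, covering the fuselage and small attached parts: a dancer figure, a paper and a booster.
# Each of the five bones gets a single capsule body.
# Limit constraints join the dancer base, paper and booster base to the fuselage, with the dancer head chained to its base.
# All joints share a tight twist limit and modest swing cones, widest on the booster.
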PhysicsAsset /Game/Athena/DrivableVehicles/Biplane/Meshes/Biplane_Phat_Dyn
bodies {
 booster_base [{shapeType=Capsule, length=15.2477, radius=6.4228, center=<0 0 0.9468>}]
 dancer_base [{shapeType=Capsule, length=6.35, radius=3.1281, center=<0 0 0.5>}]
 dancer_head [{shapeType=Capsule, length=6.35, radius=3.1281, center=<0 0 3.9659>}]
 fuselage [{shapeType=Capsule, length=5, radius=5, rotation=<rot 90 0 0>}]
 paper [{shapeType=Capsule, length=6.35, radius=3.1281, center=<1.545 -0.0001 0.3715>, rotation=<rot 90 0 0>}]
}
constraints [
 {bone1Name=dancer_base, bone2Name=fuselage, swing1LimitDegrees=20, swing2LimitDegrees=20, twistLimitDegrees=5}
 {bone1Name=dancer_head, bone2Name=dancer_base, swing1LimitDegrees=20, swing2LimitDegrees=20, twistLimitDegrees=5}
 {bone1Name=paper, bone2Name=fuselage, swing1LimitDegrees=20, swing2LimitDegrees=20, twistLimitDegrees=5}
 {bone1Name=booster_base, bone2Name=fuselage, swing1LimitDegrees=45, swing2LimitDegrees=45, twistLimitDegrees=5}
]
