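# Biplane secondary-motion AnimBP that layers procedural jiggle onto an input pose for the paper, dancer bobblehead, wings and engine pistons.
# A RigidBody node at zero alpha precedes a chain of additive bone-space rotations.
# Range-mapped SpeedABS weights paper, dancer and pistons, with three jiggle rotators varying the pistons; WingJiggleAlpha weights all four wings' shared rotation.
# The EventGraph is empty and none of the driving variables are declared here.
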AnimBlueprint /Game/Athena/DrivableVehicles/Biplane/Meshes/Biplane_AnimBP_Dyn
target_skeleton=&/Biplane_Skeleton
graphs [
 AnimGraph {nodes [
   AnimGraphNode_SubInput_0 = Input Pose()
   n1 = RigidBody(ComponentPose=Local To Component(LocalPose=@AnimGraphNode_SubInput_0), Alpha=0) {
    bOverrideWorldGravity=true
    lODThreshold=1
    overridePhysicsAsset=&/Biplane_Phat_Dyn
    alphaBoolBlend {blendOption=Linear}
    componentAppliedLinearAccClamp=<1000 1000 1000>
    componentLinearAccScale=<5 5 5>
    componentLinearVelScale=<1 1 1>
    overrideWorldGravity {z=-980}
   }
   n2 = Get SpeedABS() -> (SpeedABS:float)
   n3 = "Transform (Modify) Bone - Bone: paper"!(ComponentPose=@n1, Alpha=@n2, Rotation=Get PaperJiggle() -> (PaperJiggle:Rotator)) -> (Pose) {
    rotationMode=BMM_Additive
    rotationSpace=BCS_BoneSpace
    translationSpace=BCS_ParentBoneSpace
    alphaBoolBlend {blendOption=Linear}
    alphaScaleBias {scale=0.2}
    alphaScaleBiasClamp {bMapRange=true, interpSpeedDecreasing=0.1, interpSpeedIncreasing=0.1, inRange {max=60, min=15}}
    boneToModify {boneName=paper}
   }
   n4 = "Transform (Modify) Bone - Bone: dancer_base"!(ComponentPose=@n3, Alpha=@n2, Rotation=Get BobbleHeadJiggle() -> (BobbleHeadJiggle:Rotator)) -> (Pose) {
    rotationMode=BMM_Additive
    rotationSpace=BCS_BoneSpace
    translationSpace=BCS_ParentBoneSpace
    alphaBoolBlend {blendOption=Linear}
    alphaScaleBias {scale=0.2}
    alphaScaleBiasClamp {bMapRange=true, interpSpeedDecreasing=0.1, interpSpeedIncreasing=0.1, inRange {max=60, min=15}}
    boneToModify {boneName=dancer_base}
   }
   n5 = Get WingRotations() -> (WingRotations:Rotator)
   n6 = Get WingJiggleAlpha() -> (WingJiggleAlpha:float)
   n7 = "Transform (Modify) Bone - Bone: wing_top_l"!(ComponentPose=@n4, Alpha=@n6, Rotation=@n5) -> (Pose) {
    rotationMode=BMM_Additive
    rotationSpace=BCS_BoneSpace
    translationSpace=BCS_ParentBoneSpace
    alphaBoolBlend {blendOption=Linear}
    boneToModify {boneName=wing_top_l}
   }
   n8 = "Transform (Modify) Bone - Bone: wing_bott_l"!(ComponentPose=@n7, Alpha=@n6, Rotation=@n5) -> (Pose) {
    rotationMode=BMM_Additive
    rotationSpace=BCS_BoneSpace
    translationSpace=BCS_ParentBoneSpace
    alphaBoolBlend {blendOption=Linear}
    boneToModify {boneName=wing_bott_l}
   }
   n9 = "Transform (Modify) Bone - Bone: wing_top_r"!(ComponentPose=@n8, Alpha=@n6, Rotation=@n5) -> (Pose) {
    rotationMode=BMM_Additive
    rotationSpace=BCS_BoneSpace
    translationSpace=BCS_ParentBoneSpace
    alphaBoolBlend {blendOption=Linear}
    boneToModify {boneName=wing_top_r}
   }
   n10 = "Transform (Modify) Bone - Bone: wing_bott_r"!(ComponentPose=@n9, Alpha=@n6, Rotation=@n5) -> (Pose) {
    rotationMode=BMM_Additive
    rotationSpace=BCS_BoneSpace
    translationSpace=BCS_ParentBoneSpace
    alphaBoolBlend {blendOption=Linear}
    boneToModify {boneName=wing_bott_r}
   }
   n11 = Get CurrentDriveJiggleRandom2() -> (CurrentDriveJiggleRandom2:Rotator)
   n12 = "Transform (Modify) Bone - Bone: engine_piston_fr_l"(ComponentPose=@n10, Alpha=@n2, Translation="0,0,0", Rotation=@n11, Scale="1,1,1") {
    lODThreshold=1
    rotationMode=BMM_Additive
    rotationSpace=BCS_BoneSpace
    alphaBoolBlend {blendOption=Linear}
    alphaScaleBiasClamp {bMapRange=true, scale=0.9, inRange {max=45}, outRange {min=0.2}}
    boneToModify {boneName=engine_piston_fr_l}
   }
   n13 = Get CurrentDriveJiggleRandom() -> (CurrentDriveJiggleRandom:Rotator)
   n14 = "Transform (Modify) Bone - Bone: engine_piston_mid_l"(ComponentPose=@n12, Alpha=@n2, Translation="0,0,0", Rotation=@n13, Scale="1,1,1") {
    lODThreshold=1
    rotationMode=BMM_Additive
    rotationSpace=BCS_BoneSpace
    alphaBoolBlend {blendOption=Linear}
    alphaScaleBiasClamp {bMapRange=true, scale=1.1, inRange {max=65}, outRange {min=0.2}}
    boneToModify {boneName=engine_piston_mid_l}
   }
   n15 = Get CurrentDriveJiggle() -> (CurrentDriveJiggle:Rotator)
   n16 = "Transform (Modify) Bone - Bone: engine_piston_bk_l"(ComponentPose=@n14, Alpha=@n2, Translation="0,0,0", Rotation=@n15, Scale="1,1,1") {
    lODThreshold=1
    rotationMode=BMM_Additive
    rotationSpace=BCS_BoneSpace
    alphaBoolBlend {blendOption=Linear}
    alphaScaleBiasClamp {bMapRange=true, inRange {max=60}, outRange {min=0.2}}
    boneToModify {boneName=engine_piston_bk_l}
   }
   n17 = "Transform (Modify) Bone - Bone: engine_piston_fr_r"(ComponentPose=@n16, Alpha=@n2, Translation="0,0,0", Rotation=@n15, Scale="1,1,1") {
    lODThreshold=1
    rotationMode=BMM_Additive
    rotationSpace=BCS_BoneSpace
    alphaBoolBlend {blendOption=Linear}
    alphaScaleBiasClamp {bMapRange=true, inRange {max=60}, outRange {min=0.2}}
    boneToModify {boneName=engine_piston_fr_r}
   }
   n18 = "Transform (Modify) Bone - Bone: engine_piston_mid_r"(ComponentPose=@n17, Alpha=@n2, Translation="0,0,0", Rotation=@n11, Scale="1,1,1") {
    lODThreshold=1
    rotationMode=BMM_Additive
    rotationSpace=BCS_BoneSpace
    alphaBoolBlend {blendOption=Linear}
    alphaScaleBiasClamp {bMapRange=true, scale=0.9, inRange {max=45}, outRange {min=0.2}}
    boneToModify {boneName=engine_piston_mid_r}
   }
   n19 = "Transform (Modify) Bone - Bone: engine_piston_bk_r"(ComponentPose=@n18, Alpha=@n2, Translation="0,0,0", Rotation=@n13, Scale="1,1,1") {
    lODThreshold=1
    rotationMode=BMM_Additive
    rotationSpace=BCS_BoneSpace
    alphaBoolBlend {blendOption=Linear}
    alphaScaleBiasClamp {bMapRange=true, scale=1.1, inRange {max=65}, outRange {min=0.2}}
    boneToModify {boneName=engine_piston_bk_r}
   }
   Output Pose(Result=Component To Local(ComponentPose=@n19)) {layerGroup=None}
  ]}
 EventGraph {links [], nodes []}
]
variables []
