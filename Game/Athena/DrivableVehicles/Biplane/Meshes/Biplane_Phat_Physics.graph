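# Physics asset for the drivable Biplane vehicle in Athena, defining its simplified collision as a single rigid body on the fuselage bone.
# The body compounds four named primitives: capsules for the main body and tapering tail, plus boxes for the wings and tail flaps.
# It has no other bodies or constraints, so none of the plane's parts are simulated or articulated separately.
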PhysicsAsset /Game/Athena/DrivableVehicles/Biplane/Meshes/Biplane_Phat_Physics
bodies {fuselage [
  {shapeType=Capsule, length=127.786, radius=72.0163, shapeName=Body, center=<0.6753 -0.0002 -0.2577>, rotation=<rot 84.9999 0 0>}
  {shapeType=Capsule, length=221.0881, radius=35.4641, shapeName=Tail, center=<-190.6873 -0.0001 -3.6642>, rotation=<rot 77.2697 0 0>}
  {shapeType=Box, extentX=106.75, extentY=612.459, extentZ=104, shapeName=Wings, center=<-8.287 0 4.8098>}
  {shapeType=Box, extentX=97.5, extentY=217.5, extentZ=18.5, shapeName=TailFlaps, center=<-338.7278 0 28.0582>}
 ]}
constraints []
